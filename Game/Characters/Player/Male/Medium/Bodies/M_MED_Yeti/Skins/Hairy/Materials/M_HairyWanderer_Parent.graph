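# Parent material for the Hairy Wanderer skin of the medium male Yeti player body, masked with dithered opacity and usable on skeletal meshes and mesh deformers.
# It chains shared material-attribute functions: the character PreFX base feeds a near-distance fade, which feeds the PostFX function.
# The PreFX AO output also drives PostFX's AO, Distance Blend, Fresnel, Gmap and Pre Skinned Local Position inputs.
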Material /Game/Characters/Player/Male/Medium/Bodies/M_MED_Yeti/Skins/Hairy/Materials/M_HairyWanderer_Parent
graph {
 properties {
  bAutomaticallySetUsageInEditor=false
  bUseMaterialAttributes=true
  bUsedWithMeshDeformer=true
  bUsedWithSkeletalMesh=true
  blendMode=BLEND_Masked
  ditherOpacityMask=true
  usedShadingModels=Default Lit
 }
 nodes [
  n1 = MaterialFunctionCall() {materialFunction=&BaseFN_CharacterFunction_PreFX}
  MaterialFunctionCall(
   AO=@n1.AO
   Distance Blend=@n1.AO
   Fresnel=@n1.AO
   Gmap=@n1.AO
   Pre Skinned Local Position=@n1.AO
   Material=MaterialFunctionCall!(In=@n1.Result) {materialFunction=&/MF_KnishClamp_NearDistanceFade}.Result
  ) {materialFunction=&BaseFN_CharacterFunction_PostFX}
 ]
}
